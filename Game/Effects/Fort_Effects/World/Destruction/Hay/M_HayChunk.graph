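# Masked, two-sided lit material for mesh particles, used for hay chunks in world destruction effects.
# It samples the haystack prop textures, tinting diffuse by particle color before ConvertFromDiffSpec and softening the normal map through FlattenNormal with a Flatness parameter.
# Particle alpha multiplies the mask texture for cutout, and the diffuse red channel scaled by a Roughness parameter doubles as a roughness source, avoiding a separate map.
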Material /Game/Effects/Fort_Effects/World/Destruction/Hay/M_HayChunk
graph {
 properties {
  bEnableResponsiveAA=true
  bUsedWithMeshParticles=true
  blendMode=BLEND_Masked
  pixelDepthOffsetMode=PDOM_Legacy
  twoSided=true
  usedShadingModels=Default Lit
 }
 nodes [
  n1 = TextureSample() {texture=/Game/Environments/Props/Rural/Textures/Prop_HayStack_01_D}
  n2 = ParticleColor()
  MaterialFunctionCall(DiffuseColor=Multiply(A=@n1.RGB, B=@n2.RGB)) {materialFunction=&ConvertFromDiffSpec}
  MaterialFunctionCall(
   Normal=TextureSample() {samplerType=SAMPLERTYPE_Normal, texture=/Game/Environments/Props/Rural/Textures/Prop_HayStack_01_N}.RGB
   Flatness=ScalarParameter() {defaultValue=0.65, parameterName=Flatness}
  ) {materialFunction=&FlattenNormal}
  Multiply(A=@n2.A, B=TextureSample() {texture=/Game/Environments/Props/Rural/Textures/Prop_HayStack_01_M}.RGB)
  Multiply(A=@n1.R, B=ScalarParameter() {defaultValue=0.8, parameterName=Roughness})
 ]
}
