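# Translucent, two-sided lit particle material for hay debris, usable on Cascade and Niagara sprites.
# The SubUV hay texture is tinted by particle color, with optional time-of-day brightness and emissive; both branches of UseHayTextureColor? read the same texture RGB.
# Dynamic parameter LerpMacroIn blends alpha from a sharpened power curve to one modulated by panning smoke noise on macro UVs.
# Particle alpha scales the result, with optional near-camera fade that mobile bypasses.
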Material /Game/Effects/Fort_Effects/World/Destruction/Hay/M_HayDebris
graph {
 properties {
  bEnableResponsiveAA=true
  bUsedWithNiagaraSprites=true
  bUsedWithParticleSprites=true
  blendMode=BLEND_Translucent
  pixelDepthOffsetMode=PDOM_Legacy
  translucencyLightingMode=TLM_VolumetricPerVertexNonDirectional
  twoSided=true
  usedShadingModels=Default Lit
 }
 nodes [
  MaterialFunctionCall() {materialFunction=&WorldAlignedTexture}
  n1 = ParticleColor()
  n2 = ScalarParameter() {parameterName=Emissive_Amount}
  MaterialFunctionCall(Emissive=StaticSwitchParameter(True=Multiply(A=@n1.RGB, B=@n2), False=Constant()) {parameterName=UseEmissive?}) {materialFunction=&MF_FX_EmissiveAdjustments}
  Multiply()
  n3 = ScalarParameter() {defaultValue=1, group=TimeOfDay, parameterName=DayTimeBrightness}
  n4 = ScalarParameter() {defaultValue=0.5, group=TimeOfDay, parameterName=NightBrightness}
  n5 = ParticleSubUV!(UVs, Tex, Apply View MipBias) {texture=&/T_HayDebris}
  n6 = ParticleColor()
  n7 = MaterialFunctionCall(DiffuseColor=Multiply(A=StaticSwitchParameter(True=@n5.RGB, False=@n5.RGB) {defaultValue=true, parameterName=UseHayTextureColor?}, B=@n6.RGB)) {materialFunction=&ConvertFromDiffSpec}
  StaticSwitchParameter(
   True=Multiply(
    A=MaterialFunctionCall(Color_Morning=@n3, Color_Day=@n3, Color_Evening=@n4, Color_Night=@n4) {materialFunction=&Color_Time_of_Day}.Result
    B=@n7.BaseColor
   )
   False=@n7.BaseColor
  ) {group=TimeOfDay, parameterName=UseTimeOfDay}
  n8 = ParticleMacroUV()
  n9 = LinearInterpolate(Alpha=TextureSample(UVs=Panner(Coordinate=@n8) {speedY=-0.06}) {texture=/Game/Effects/Fort_Effects/Textures/T_Limbo_smoke_pool}.R) {constA=0.5, constB=1.25}
  n10 = Multiply(
   A=LinearInterpolate(
    A=Power(Base=@n5.A) {constExponent=6}
    B=Multiply(A=@n5.A, B=@n9)
    Alpha=DynamicParameter() {defaultValue=<rgba 0 0 0 1>, paramNames [LerpMacroIn,Param2,Param3,Param4]}.LerpMacroIn
   ) {constAlpha=1}
   B=@n6.A
  )
  n11 = MaterialFunctionCall(Distance=ScalarParameter() {defaultValue=256, parameterName="CameraFadeDistance "}) {materialFunction=&Near_Camera_Fade}
  n12 = Constant() {r=1}
  StaticSwitchParameter(True=Multiply(A=@n10, B=ShadingPathSwitch(Default=@n11.Result, Mobile=@n12) {desc=Migrated from FeatureLevelSwitch}), False=@n10) {parameterName=UseCameraFade}
 ]
}
